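# MetaSound Source for the Major Spire boss's look-at movement loop, a spatialized stereo foley sound.
# Health indexes a wave array on play or health change, restarting the looping player at a random offset; all default slots hold the idle loop.
# Smoothed movement speed within a threshold band gates an ADSR envelope, with trigger-once latches and a delayed release.
# That envelope raises the loop's volume and pitch.
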MetaSoundSource /SpireBoss/Audio/MetaSounds/MSS_MajorSpire_LookAt_Loop
inputs [
 Health {type=Int32}
 Movement.Speed {type=Float}
 Movement.SpeedAzimuth {type=Float}
 Movement.SpeedElevation {type=Float}
 Sounds {type=WaveAsset:Array}
 UE.Source.OnPlay {type=Trigger}
]
interfaces [Fort.Movement:1.0, UE.OutputFormat.Stereo:1.0, UE.Source:1.1]
outputs [
 "UE.OutputFormat.Stereo.Audio:0" {type=Audio}
 "UE.OutputFormat.Stereo.Audio:1" {type=Audio}
]
pages [Default {
  input_defaults {
   Health=0
   Movement.Speed=26.3999
   Movement.SpeedAzimuth=0
   Movement.SpeedElevation=0
   UE.Source.OnPlay=null
   Sounds [/SpireBoss/Audio/Waves/MajorSpire_Idle_Loop, &MajorSpire_Idle_Loop, &MajorSpire_Idle_Loop]
  }
  nodes [
   InitVariable.Float()
   n1 = UE.InterpTo:Audio(Target=^Movement.Speed) {also {defaults {Interp Time=1}}}
   TriggerOnValueChange_0 = UE.Trigger On Value Change:Float(Value=@n1.Value)
   n2 = TriggerCompare.Clamp:Float(A=@n1.Value, Compare=@TriggerOnValueChange_0.Trigger) {also {defaults {B=18, Type=3}}}
   n3 = TriggerCompare.Clamp:Float(A=@n1.Value, Compare=@n2.True) {also {defaults {B=33, Type=2}}}
   TriggerAnyTriggerAny2_0 = "TriggerAny.Trigger Any (2)"(In 0=UE.Reroute(Value=UE.Reroute(Value=@n2.False).Value).Value, In 1=@n3.False)
   n4 = VariableDeferredAccessor.Trigger(Variable=InitVariable.Trigger().Variable)
   TriggerAnyTriggerAny2_2 = "TriggerAny.Trigger Any (2)"(In 1=@TriggerAnyTriggerAny2_0.Out, In 0=@n4.Value)
   TriggerControl_0 = UE.Trigger Control(Close=@TriggerAnyTriggerAny2_0.Out, Open=@n3.True, Trigger In=@TriggerOnValueChange_0.Trigger)
   TriggerOnce_0 = UE.Trigger Once(Trigger In=@TriggerAnyTriggerAny2_2.Out, Reset=@TriggerControl_0.Trigger Out)
   TriggerDelay_0 = UE.Trigger Delay(In=@TriggerOnce_0.Trigger Out) {also {defaults {Delay Time=0.5}}}
   TriggerOnce_1 = UE.Trigger Once(Trigger In=@TriggerControl_0.Trigger Out, Reset=@TriggerDelay_0.Out)
   ADSREnvelopeADSREnvelope_0 = ADSR Envelope.ADSR Envelope:Float(Trigger Release=@TriggerDelay_0.Out, Trigger Attack=@TriggerOnce_1.Trigger Out) {also {defaults {Attack Time=1, Sustain Level=1}}}
   n5 = VariableAccessor.Float(Variable=VariableMutator.Float(Value=@ADSREnvelopeADSREnvelope_0.Out Envelope, Variable=InitVariable.Float().Variable).Variable)
   n6 = MapRange.MapRange:Float(In=@n5.Value) {also {defaults {Out Range A=0.5}}}
   TriggerOnValueChange_1 = UE.Trigger On Value Change:Int32(Value=^Health)
   TriggerAnyTriggerAny2_1 = "TriggerAny.Trigger Any (2)"(In 0=^UE.Source.OnPlay, In 1=@TriggerOnValueChange_1.Trigger)
   n7 = Array.Get:WaveAsset:Array(Index=^Health, Array=^Sounds, Trigger=@TriggerAnyTriggerAny2_1.Out)
   FrequencyMultipliertoSemitone_0 = UE.Frequency Multiplier to Semitone:Float(Frequency Multiplier=MapRange.MapRange:Float(In=@n5.Value) {also {defaults {Out Range A=0.95, Out Range B=1.2}}}.Out Value)
   GetWaveDuration_0 = "UE.Get Wave Duration: "(Wave=@n7.Element)
   n8 = UE.RandomTime(Max=@GetWaveDuration_0.Duration, Next=@TriggerAnyTriggerAny2_1.Out)
   WavePlayer_0 = UE.Wave Player:Stereo(Wave Asset=@n7.Element, Pitch Shift=@FrequencyMultipliertoSemitone_0.Semitones, Play=@n8.On Next, Start Time=@n8.Value) {also {defaults {Loop="<MetaSoundBuilderResult.SUCCEEDED: 0>"}}}
   Multiply_0 = UE.Multiply:Audio by Float(AdditionalOperands=@n6.Out Value, PrimaryOperand=@WavePlayer_0.Out Left)
   Multiply_1 = UE.Multiply:Audio by Float(AdditionalOperands=@n6.Out Value, PrimaryOperand=@WavePlayer_0.Out Right)
   VariableMutator.Trigger(Value=@TriggerOnValueChange_0.Trigger, Variable=@n4.Variable)
  ]
  links [{dst="@UE.OutputFormat.Stereo.Audio:0", src=Multiply_0.Out}, {dst="@UE.OutputFormat.Stereo.Audio:1", src=Multiply_1.Out}]
 }]
properties {
 assetImportData=&self:AssetImportData
 attenuationSettings=/SpireBoss/Audio/Settings/ATT_MajorSpire_Movement_Small
 bLooping=true
 duration=10000
 maxDistance=10500
 numChannels=2
 outputFormat=Stereo
 priority=2
 sampleRate=48000
 soundClassObject=/Game/Sounds/FortSoundClassesAndModes/Fort_BossFight_SoundClass/Fort_BossFight_Foley
 volume=2
 concurrencySet [/SpireSharedAssets/Audio/Settings/Conc_MajorSpire_Ambient, /SpireSharedAssets/Audio/Settings/Conc_MajorSpire_Ambient_Lightning]
 modulationSettings {
  volumeRouting=Union
  volumeModulationDestination {bEnableModulation=true, modulators [/SpireSharedAssets/Audio/Modulation/PP_Spires_Loops_Vol]}
 }
}
