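# One-shot stereo MetaSound source for the Major Spire boss projectile spawn-in sound.
# On play, the recent-play count modulo four picks one of four rising pitch values, so successive spawns step upward and then wrap.
# That pitch triggers a shared random-sound player with slight volume variation.
# Boss-fight sound class, attenuation, concurrency and sidechain bus sends for loops, weapons, music and ambience set its place in the mix.
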MetaSoundSource /SpireBoss/Audio/MetaSounds/MSS_MajorSpire_Proj_SpawnIn
inputs [
 RecentCollection.NumRecentPlays {type=Int32}
 RecentCollection.TimeSinceLastPlayed {type=Float}
 UE.Source.OnPlay {type=Trigger}
]
interfaces [Fort.ActiveSoundRecentCollection:1.0, UE.OutputFormat.Stereo:1.0, UE.Source.OneShot:1.0, UE.Source:1.1]
outputs [
 "UE.OutputFormat.Stereo.Audio:0" {type=Audio}
 "UE.OutputFormat.Stereo.Audio:1" {type=Audio}
 UE.Source.OneShot.OnFinished {type=Trigger}
]
pages [Default {
  input_defaults {RecentCollection.NumRecentPlays=0, RecentCollection.TimeSinceLastPlayed=0, UE.Source.OnPlay=null}
  nodes [
   TriggerSelect4_0 = "UE.Trigger Select (4)"(In=^UE.Source.OnPlay, Index=UE.Modulo:Int32(PrimaryOperand=^RecentCollection.NumRecentPlays) {also {defaults {AdditionalOperands=4}}}.Out)
   TriggerRouteTriggerRouteFloat4_0 = "TriggerRoute.Trigger Route (Float, 4):Float"(Set 0=@TriggerSelect4_0.Out 0, Set 1=@TriggerSelect4_0.Out 1, Set 2=@TriggerSelect4_0.Out 2, Set 3=@TriggerSelect4_0.Out 3) {also {defaults {Value 0=1, Value 1=1.25, Value 2=1.5, Value 3=1.75}}}
   MSS_PlayRandomSound_Stereo_0 = &MSS_PlayRandomSound_Stereo(
    UE.Source.OnPlay=@TriggerRouteTriggerRouteFloat4_0.On Set
    Pitch=@TriggerRouteTriggerRouteFloat4_0.Value
    Sounds=VariableAccessor.WaveAsset:Array(Variable=InitVariable.WaveAsset:Array().Variable).Value
   ) {also {defaults {MaxVolume=1.05, MinVolume=0.95}}}
  ]
  links [
   {dst="@UE.OutputFormat.Stereo.Audio:0", src=MSS_PlayRandomSound_Stereo_0.UE.OutputFormat.Stereo.Audio:0}
   {dst="@UE.OutputFormat.Stereo.Audio:1", src=MSS_PlayRandomSound_Stereo_0.UE.OutputFormat.Stereo.Audio:1}
   {dst="@UE.Source.OneShot.OnFinished", src=MSS_PlayRandomSound_Stereo_0.UE.Source.OneShot.OnFinished}
  ]
 }]
properties {
 assetImportData=&self:AssetImportData
 attenuationSettings=/SpireBoss/Audio/Settings/ATT_MajorSpire_Proj_Spawn
 bBypassVolumeScaleForPriority=true
 maxDistance=9500
 numChannels=2
 outputFormat=Stereo
 sampleRate=48000
 soundClassObject=/Game/Sounds/FortSoundClassesAndModes/Fort_BossFight_SoundClass/Fort_BossFight_PlayerCritical
 volume=0.9
 busSends [
  {audioBus=/SpireSharedAssets/Audio/Modulation/AB_Spires_Loops_SideChain_Vol, sendLevel=0.8}
  {audioBus=&AB_Sidechain_FortWeapon_Vol, sendLevel=0.7}
  {audioBus=&AB_Sidechain_FortMusic_Vol, sendLevel=0.4}
  {audioBus=&AB_Sidechain_FortAmbient_Vol, sendLevel=0.8}
 ]
 concurrencySet [/SpireBoss/Audio/Settings/Conc_MajorSpire_Proj]
}
